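# Animation Blueprint for the male medium AuroraOwl player body cosmetic, layering Control Rig on the pose copied from its attached parent mesh.
# The AnimGraph feeds that pose into M_MED_AuroraOwl_CtrlRig with skydive, in-air, emote, crouch, DBNO and parachute states.
# Each update, the EventGraph converts skydive and emote booleans to floats and sets in-air to one while skydiving or skydive-floating, otherwise zero.
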AnimBlueprint /BRCosmetics/Characters/Player/Male/Medium/Bodies/M_MED_AuroraOwl/Meshes/M_MED_AuroraOwl_AnimBP
target_skeleton=&/M_MED_AuroraOwl_Skeleton
graphs [
 AnimGraph {nodes [
   n1 = Control Rig!(
    Source=Copy Pose From Mesh() {bUseAttachedParent=true}
    Alpha
    is_skydive:double=Get skydiveFloat() -> (skydiveFloat:double)
    is_in_air:double=Get Is in Air() -> (Is in Air:double)
    is_playing_emote:double=Get Is Playing Emote() -> (Is Playing Emote:double)
    is_crouch:bool=Get bIsCrouching()
    is_DBNO:bool=Get bIsDBNO()
    is_parachuting:bool=Get bIsParachuteOpened()
   ) -> (Pose) {
    alphaBoolBlend {blendOption=Linear}
    controlRigAssetReference {blueprintRigClass=&/M_MED_AuroraOwl_CtrlRig.M_MED_AuroraOwl_CtrlRig_C}
   }
   Output Pose(Result=@n1) {layerGroup=None}
  ]}
 EventGraph {nodes [
   Event BlueprintUpdateAnimation()
   > n1 = Sequence!(execute) -> (then_0, then_1, then_2:Exec)
   Set Is Playing Emote(execute=@n1.then_2, Is Playing Emote:double="To Float (Boolean)"(InBool=Get bIsPlayingEmote())) -> (Output_Get:double)
   Set skydiveFloat(execute=@n1.then_1, skydiveFloat:double="To Float (Boolean)"(InBool=Get bIsSkyDiving())) -> (Output_Get:double)
   n2 = Branch(execute=@n1.then_0, Condition=OR Boolean(A=Get bIsSkydiveFloating(), B=Get bIsSkyDiving()))
   > Set Is in Air(Is in Air:double=1) -> (Output_Get:double)
   Set Is in Air(execute=@n2.else, Is in Air=0)
  ]}
]
variables [
 Is Crouching {type=boolean}
 Is DBNO {type=boolean}
 Is in Air {type=number}
 Is Parachuting {type=boolean}
 Is Playing Emote {type=number}
 skydiveFloat {type=number}
]
